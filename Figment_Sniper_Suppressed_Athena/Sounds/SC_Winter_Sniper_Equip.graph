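# Sound cue for the equip foley of the suppressed winter scoped sniper rifle in the Figment Athena weapon set.
# A fixed short delay feeds a single wave player that reuses the winter sniper's reload-end sample rather than a dedicated equip sound.
# It plays at reduced volume through the weapon foley sound class with shared reload attenuation, a concurrency override and debug enabled.
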SoundCue /Figment_Sniper_Suppressed_Athena/Sounds/SC_Winter_Sniper_Equip
root=Delay_0
nodes [
 {id=Delay_0, children [WavePlayer_2], properties {delayMax=0.1, delayMin=0.1}}
 {id=WavePlayer_2, properties {soundWaveAssetPtr=/Game/Sounds/Fort_Weapons/SniperRifles/WinterScoped/Winter_Sniper_ReloadEnd_01}}
]
properties {
 attenuationSettings=&Fort_Weapon_Reload_Att
 bDebug=true
 bOverrideConcurrency=true
 maxDistance=2250
 soundClassObject=&Fort_WeaponFoley
 volumeMultiplier=0.7
}
