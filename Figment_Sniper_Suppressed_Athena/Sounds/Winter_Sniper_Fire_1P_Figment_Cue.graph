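# First-person fire SoundCue for the Figment suppressed sniper, built from Winter scoped sniper waves.
# A mixer layers one of three random close-fire shots with one of three random eject-impact sounds, the latter delayed over half a second, much quieter and pitch-varied.
# A root modulator adds slight volume variation to the boosted cue.
# First-person attenuation and owner-limited concurrency that stops the oldest instance govern playback, and the debug flag is on.
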SoundCue /Figment_Sniper_Suppressed_Athena/Sounds/Winter_Sniper_Fire_1P_Figment_Cue
root=Modulator_0
nodes [
 {id=Delay_0, children [Random_8], properties {delayMax=0.7, delayMin=0.6}}
 {id=Mixer_2, children [Random_6,Modulator_1], properties {inputVolume [1,1]}}
 {id=Modulator_0, children [Mixer_2], properties {volumeMax=1, volumeMin=0.9}}
 {id=Modulator_1, children [Delay_0], properties {pitchMax=1, pitchMin=0.9, volumeMax=0.2, volumeMin=0.1}}
 {id=Random_6, children [WavePlayer_13, WavePlayer_14, WavePlayer_15], properties {bShouldExcludeFromBranchCulling=true, weights [1,1,1]}}
 {id=Random_8, children [WavePlayer_22, WavePlayer_23, WavePlayer_24], properties {bShouldExcludeFromBranchCulling=true, weights [1,1,1]}}
 {id=WavePlayer_13, properties {soundWaveAssetPtr=/Game/Sounds/Fort_Weapons/SniperRifles/WinterScoped/Winter_Sniper_FireClose_01}}
 {id=WavePlayer_14, properties {soundWaveAssetPtr=/Game/Sounds/Fort_Weapons/SniperRifles/WinterScoped/Winter_Sniper_FireClose_02}}
 {id=WavePlayer_15, properties {soundWaveAssetPtr=/Game/Sounds/Fort_Weapons/SniperRifles/WinterScoped/Winter_Sniper_FireClose_03}}
 {id=WavePlayer_22, properties {soundWaveAssetPtr=/Game/Sounds/Fort_Weapons/SniperRifles/WinterScoped/Winter_Sniper_ReloadEjectImpact_01}}
 {id=WavePlayer_23, properties {soundWaveAssetPtr=/Game/Sounds/Fort_Weapons/SniperRifles/WinterScoped/Winter_Sniper_ReloadEjectImpact_02}}
 {id=WavePlayer_24, properties {soundWaveAssetPtr=/Game/Sounds/Fort_Weapons/SniperRifles/WinterScoped/Winter_Sniper_ReloadEjectImpact_03}}
]
properties {
 attenuationSettings=&Fort_Weapon_Projectile_FirstPerson_Athena_Att
 bDebug=true
 maxDistance=2097152
 soundClassObject=&Fort_WeaponRanged_Fire_1P_PSM
 volumeMultiplier=1.5
 concurrencyOverrides {bLimitToOwner=true, maxCount=3, resolutionRule=StopOldest}
 concurrencySet [&Weapon_Gunfire_FirstPerson_Conc]
}
